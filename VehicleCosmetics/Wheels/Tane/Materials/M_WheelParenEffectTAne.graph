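# Translucent material for an animated, cloudy effect on the Tane vehicle wheel cosmetic, usable on skeletal meshes.
# Two counter-rotating noise samples slightly distort the UVs of panning cloud and distortion textures, which combine and are shaped by UV gradient masks.
# The result drives emissive, tinted by Custom_Color and Brightness, and opacity, scaled by vertex alpha.
# These attributes pass through the shared vehicle damage and de-rez transition functions.
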Material /VehicleCosmetics/Wheels/Tane/Materials/M_WheelParenEffectTAne
graph {
 properties {
  bAutomaticallySetUsageInEditor=false
  bUseMaterialAttributes=true
  bUsedWithMeshDeformer=true
  bUsedWithSkeletalMesh=true
  blendMode=BLEND_Translucent
  usedShadingModels=Default Lit
 }
 nodes [
  n1 = VectorParameter() {parameterName=Custom_Color, defaultValue=<rgba 0 0 0 1>}
  n2 = Multiply(A=Time(), B=Constant() {r=0.5})
  n3 = Rotator(Coordinate=TextureCoordinate() {uTiling=0.5, vTiling=0.5}, Time=@n2)
  n4 = Rotator(Coordinate=TextureCoordinate(), Time=@n2) {speed=-0.25}
  n5 = Multiply(
   A=AppendVector(
    A=TextureSample(UVs=@n3) {samplerType=SAMPLERTYPE_LinearColor, texture=/VehicleCosmetics/Wheels/_SharedTextures/FX/CloudyNoise_T}.G
    B=TextureSample(UVs=@n4) {texture=/VehicleCosmetics/Wheels/_SharedTextures/FX/Water_01_Pack}.G
   )
   B=Constant() {r=0.01}
  )
  n6 = Add(A=TextureCoordinate() {uTiling=3, vTiling=0.15}, B=@n5)
  n7 = TextureSample(UVs=Panner(Coordinate=@n6) {speedX=0.25, speedY=4}) {samplerType=SAMPLERTYPE_LinearColor, texture=&CloudyNoise_T}
  n8 = Constant() {r=2}
  n9 = TextureCoordinate() {uTiling=4, vTiling=0.25}
  n10 = TextureSample(UVs=Panner(Coordinate=Add(A=@n9, B=@n5)) {speedX=0.1, speedY=2}) {samplerType=SAMPLERTYPE_LinearColor, texture=&CloudyNoise_T}
  n11 = TextureSample(UVs=Panner(Coordinate=@n6) {speedX=0.1, speedY=2}) {texture=/VehicleCosmetics/Wheels/_SharedTextures/FX/CloudDistort_T}
  n12 = Constant() {r=0.5}
  n13 = Multiply(A=Add(A=Power(Base=@n7.R, Exp=@n8), B=@n10.G), B=Add(A=@n10.R, B=Multiply(A=@n11.RGB, B=@n12)))
  n14 = Power(Base=ComponentMask(None=TextureCoordinate()) {g=true}, Exp=Constant() {r=2})
  n15 = Multiply(A=@n14, B=Constant() {r=1})
  n16 = ComponentMask(None=TextureCoordinate()) {r=true}
  n17 = Power(Base=OneMinus(None=@n16), Exp=Constant() {r=1.5})
  n18 = Power(Base=@n16, Exp=Constant() {r=1.5})
  n19 = Multiply(A=@n15, B=Multiply(A=@n15, B=Multiply(A=@n17, B=@n18)))
  n20 = Clamp(None=Multiply(A=Power(Base=@n13), B=Add(A=@n19, B=@n14)))
  n21 = Multiply(A=Multiply(A=@n20, B=Constant() {r=2}), B=Constant() {r=2})
  n22 = Clamp(None=Multiply(A=Multiply(A=@n1.RGB, B=@n21), B=ScalarParameter() {defaultValue=1, parameterName=Brightness}))
  n23 = Multiply(A=@n20, B=VertexColor().A)
  n24 = Constant() {r=1}
  n25 = Clamp(None=Multiply(A=Power(Base=@n23, Exp=@n24), B=Constant() {r=1}))
  MaterialFunctionCall(In=MaterialFunctionCall(Material=SetMaterialAttributes!(MaterialAttributes, Emissive Color=@n22, Opacity=@n25) {attributeSetTypes [#1, #2]}) {materialFunction=&MF_BaseVehicle_Damage}.Result) {materialFunction=&MF_VehicleDeRez_Transition}
 ]
}
